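# Animation blueprint for the SaladDressing face accessory mesh, adding secondary physics motion to its dyn_lock bones.
# The pose copied from the attached parent passes through replace-mode bone edits for upper arms and face, the latter gated by bHasFaceBones.
# A chain of ten root-relative Anim Dynamics nodes then produces the output.
# Each update sets a gravity override via a deprecated function and caches parent-space bone transforms from the customization's body and head meshes.
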AnimBlueprint /BRCosmetics/Characters/Player/Male/Medium/Bodies/M_MED_SaladDressing/Meshes/Parts/M_MED_SaladDressing_FaceAcc_AnimBP
target_skeleton=&/M_MED_SaladDressing_FaceAcc_Skeleton
graphs [
 AnimGraph {nodes [
   n1 = Anim Dynamics(Alpha=1, ExternalForce="0,0,0") {
    angularDampingOverride=0.95
    angularSpringConstant=1500
    bOverrideAngularDamping=true
    bUseGravityOverride=true
    gravityScale=0.35
    simulationSpace=RootRelative
    alphaBoolBlend {blendOption=Linear}
    componentAppliedLinearAccClamp=<150 150 150>
    componentLinearAccScale=<0.2 0.2 0.2>
    componentLinearVelScale=<0.2 0.2 0.2>
    gravityOverride {z=-50}
    physicsBodyDefinitions [{
      sphereCollisionRadius=0
      boundBone {boneName=None}
      boxExtents=<12 15 15>
      constraintSetup {angularTargetAxis=AxisY, angularLimitsMax=<0 10 5>, angularLimitsMin=<0 -10 -5>, angularTarget=<0 -10 0>}
      localJointOffset=<1.5 0 0>
     }]
    boundBone=~
   }
   n2 = Anim Dynamics(ComponentPose=@n1, Alpha=1, ExternalForce="0,0,0") {
    angularDampingOverride=0.8
    angularSpringConstant=1500
    bOverrideAngularDamping=true
    bUseGravityOverride=true
    gravityScale=0.35
    simulationSpace=RootRelative
    alphaBoolBlend {blendOption=Linear}
    componentAppliedLinearAccClamp=<100 100 100>
    componentLinearAccScale=<0.2 0.2 0.2>
    componentLinearVelScale=<0.2 0.2 0.2>
    gravityOverride {z=-150}
    physicsBodyDefinitions [{
      sphereCollisionRadius=0
      boundBone {boneName=None}
      boxExtents=<10 15 15>
      constraintSetup {angularTargetAxis=AxisY, angularLimitsMax=<0 10 5>, angularLimitsMin=<0 -10 -5>, angularTarget=<0 -10 0>}
      localJointOffset=<1 0 0>
     }]
    boundBone=~
   }
   Anim Dynamics(ComponentPose=@n2, Alpha=1, ExternalForce="0,0,0") {
    angularDampingOverride=0.9
    angularSpringConstant=1500
    bOverrideAngularDamping=true
    bUseGravityOverride=true
    gravityScale=0.35
    simulationSpace=RootRelative
    alphaBoolBlend {blendOption=Linear}
    componentAppliedLinearAccClamp=<200 200 200>
    componentLinearAccScale=<0.2 0.2 0.2>
    componentLinearVelScale=<0.2 0.2 0.2>
    gravityOverride {z=-50}
    physicsBodyDefinitions [{
      sphereCollisionRadius=0
      boundBone {boneName=None}
      boxExtents=<12 15 15>
      constraintSetup {angularTargetAxis=AxisY, angularLimitsMax=<0 10 5>, angularLimitsMin=<0 -10 -5>, angularTarget=<0 -10 0>}
      localJointOffset=<4 0 0>
     }]
    boundBone=~
   }
   n3 = Anim Dynamics - Bone: dyn_lock_01(ComponentPose=Local To Component(LocalPose=Use cached pose 'PostFaceUpdate'()), Alpha=1, ExternalForce="0,0,0") {
    angularDampingOverride=0.8
    angularSpringConstant=3
    bAngularSpring=true
    bOverrideAngularDamping=true
    bUseGravityOverride=true
    gravityScale=0.35
    simulationSpace=RootRelative
    alphaBoolBlend {blendOption=Linear}
    boundBone {boneName=dyn_lock_01}
    componentAppliedLinearAccClamp=<600 600 600>
    componentLinearAccScale=<0.15 0.15 0.15>
    componentLinearVelScale=<0.15 0.15 0.15>
    gravityOverride {z=-300}
    physicsBodyDefinitions [{
      sphereCollisionRadius=0
      boundBone {boneName=dyn_lock_01}
      boxExtents=<8 10 10>
      constraintSetup {angularLimitsMax=<3 3 0>, angularLimitsMin=<-3 -3 -40>, angularTarget=<0 0 0>}
      localJointOffset=<1 0.05 0>
     }]
   }
   n4 = Anim Dynamics - Bone: dyn_lock_02!(ComponentPose=@n3, Alpha=1, GravityOverride:Vector="0,0,-980", ExternalForce="0,0,0") -> (Pose) {
    angularDampingOverride=1
    angularSpringConstant=1500
    bOverrideAngularDamping=true
    bUseGravityOverride=true
    gravityScale=0.35
    simulationSpace=RootRelative
    alphaBoolBlend {blendOption=Linear}
    boundBone {boneName=dyn_lock_02}
    componentAppliedLinearAccClamp=<350 350 350>
    componentLinearAccScale=<0.2 0.2 0.2>
    componentLinearVelScale=<0.2 0.2 0.2>
    gravityOverride {z=-980}
    physicsBodyDefinitions [{
      sphereCollisionRadius=0
      boundBone {boneName=dyn_lock_02}
      boxExtents=<10 15 15>
      constraintSetup {angularTargetAxis=AxisY, angularLimitsMax=<0 0 5>, angularLimitsMin=<0 -10 -15>, angularTarget=<0 0 0>}
      localJointOffset=<2 0 0>
     }]
   }
   n5 = Anim Dynamics - Bone: dyn_lock_03(ComponentPose=@n4, Alpha=1, ExternalForce="0,0,0") {
    angularDampingOverride=0.75
    angularSpringConstant=1500
    bOverrideAngularDamping=true
    bUseGravityOverride=true
    gravityScale=0.35
    simulationSpace=RootRelative
    alphaBoolBlend {blendOption=Linear}
    boundBone {boneName=dyn_lock_03}
    componentAppliedLinearAccClamp=<500 500 500>
    componentLinearAccScale=<0.2 0.2 0.2>
    componentLinearVelScale=<0.2 0.2 0.2>
    gravityOverride {z=-250}
    physicsBodyDefinitions [{
      sphereCollisionRadius=0
      boundBone {boneName=dyn_lock_03}
      boxExtents=<10 15 15>
      constraintSetup {angularTargetAxis=AxisY, angularLimitsMax=<0 0 10>, angularLimitsMin=<0 -5 -10>, angularTarget=<0 0 0>}
      localJointOffset=<4 0 0>
     }]
   }
   n6 = Anim Dynamics - Bone: dyn_lock_04(ComponentPose=@n5, Alpha=1, ExternalForce="0,0,0") {
    angularDampingOverride=0.75
    angularSpringConstant=1500
    bOverrideAngularDamping=true
    bUseGravityOverride=true
    gravityScale=0.35
    simulationSpace=RootRelative
    alphaBoolBlend {blendOption=Linear}
    boundBone {boneName=dyn_lock_04}
    componentAppliedLinearAccClamp=<700 700 700>
    componentLinearAccScale=<0.2 0.2 0.2>
    componentLinearVelScale=<0.2 0.2 0.2>
    gravityOverride {z=-200}
    physicsBodyDefinitions [{
      sphereCollisionRadius=0
      boundBone {boneName=dyn_lock_04}
      boxExtents=<10 15 15>
      constraintSetup {angularTargetAxis=AxisY, angularLimitsMax=<0 0 5>, angularLimitsMin=<0 -15 -5>, angularTarget=<0 0 0>}
      localJointOffset=<2 0 0>
     }]
   }
   n7 = Anim Dynamics - Bone: dyn_lock_05!(ComponentPose=@n6, Alpha=1, GravityOverride:Vector="0,0,-100", ExternalForce="0,0,0") -> (Pose) {
    angularDampingOverride=0.9
    angularSpringConstant=1500
    bOverrideAngularDamping=true
    bUseGravityOverride=true
    gravityScale=0.35
    simulationSpace=RootRelative
    alphaBoolBlend {blendOption=Linear}
    boundBone {boneName=dyn_lock_05}
    componentAppliedLinearAccClamp=<100 100 100>
    componentLinearAccScale=<0.2 0.2 0.2>
    componentLinearVelScale=<0.2 0.2 0.2>
    gravityOverride {z=-100}
    physicsBodyDefinitions [{
      sphereCollisionRadius=0
      boundBone {boneName=dyn_lock_05}
      boxExtents=<10 15 15>
      constraintSetup {angularTargetAxis=AxisY, angularLimitsMax=<0 0 5>, angularLimitsMin=<0 -10 -5>, angularTarget=<0 0 0>}
      localJointOffset=<4 0 0>
     }]
   }
   n8 = Anim Dynamics - Bone: dyn_lock_06(ComponentPose=@n7, Alpha=1, ExternalForce="0,0,0") {
    angularDampingOverride=0.75
    angularSpringConstant=1500
    bOverrideAngularDamping=true
    bUseGravityOverride=true
    gravityScale=0.35
    simulationSpace=RootRelative
    alphaBoolBlend {blendOption=Linear}
    boundBone {boneName=dyn_lock_06}
    componentAppliedLinearAccClamp=<350 350 350>
    componentLinearAccScale=<0.2 0.2 0.2>
    componentLinearVelScale=<0.2 0.2 0.2>
    gravityOverride {z=-200}
    physicsBodyDefinitions [{
      sphereCollisionRadius=0
      boundBone {boneName=dyn_lock_06}
      boxExtents=<10 15 15>
      constraintSetup {angularTargetAxis=AxisY, angularLimitsMax=<0 0 0>, angularLimitsMin=<0 -15 -10>, angularTarget=<0 0 0>}
      localJointOffset=<2 0 0>
     }]
   }
   n9 = Anim Dynamics - Bone: dyn_lock_07(ComponentPose=@n8, Alpha=1, ExternalForce="0,0,0") {
    angularDampingOverride=1
    angularSpringConstant=1500
    bOverrideAngularDamping=true
    bUseGravityOverride=true
    gravityScale=0.35
    simulationSpace=RootRelative
    alphaBoolBlend {blendOption=Linear}
    boundBone {boneName=dyn_lock_07}
    componentAppliedLinearAccClamp=<3000 3000 3000>
    componentLinearAccScale=<0.2 0.2 0.2>
    componentLinearVelScale=<0.2 0.2 0.2>
    gravityOverride {z=-200}
    physicsBodyDefinitions [{
      sphereCollisionRadius=0
      boundBone {boneName=dyn_lock_07}
      boxExtents=<10 15 15>
      constraintSetup {angularTargetAxis=AxisY, angularLimitsMax=<0 0 15>, angularLimitsMin=<0 -10 -15>, angularTarget=<0 0 0>}
      localJointOffset=<1 0 0>
     }]
   }
   n10 = Anim Dynamics - Bone: dyn_lock_08!(ComponentPose=@n9, Alpha=1, GravityOverride:Vector="0,0,-980", ExternalForce="0,0,0") -> (Pose) {
    angularDampingOverride=1
    angularSpringConstant=1500
    bOverrideAngularDamping=true
    bUseGravityOverride=true
    gravityScale=0.35
    simulationSpace=RootRelative
    alphaBoolBlend {blendOption=Linear}
    boundBone {boneName=dyn_lock_08}
    componentAppliedLinearAccClamp=<350 350 350>
    componentLinearAccScale=<0.2 0.2 0.2>
    componentLinearVelScale=<0.2 0.2 0.2>
    gravityOverride {z=-980}
    physicsBodyDefinitions [{
      sphereCollisionRadius=0
      boundBone {boneName=dyn_lock_08}
      boxExtents=<10 15 15>
      constraintSetup {angularTargetAxis=AxisY, angularLimitsMax=<0 0 10>, angularLimitsMin=<0 -10 -10>, angularTarget=<0 0 0>}
      localJointOffset=<2 0 0>
     }]
   }
   n11 = Anim Dynamics - Bone: dyn_lock_09(ComponentPose=@n10, Alpha=1, ExternalForce="0,0,0") {
    angularDampingOverride=0.75
    angularSpringConstant=1500
    bOverrideAngularDamping=true
    bUseGravityOverride=true
    gravityScale=0.35
    simulationSpace=RootRelative
    alphaBoolBlend {blendOption=Linear}
    boundBone {boneName=dyn_lock_09}
    componentAppliedLinearAccClamp=<200 200 200>
    componentLinearAccScale=<0.2 0.2 0.2>
    componentLinearVelScale=<0.2 0.2 0.2>
    gravityOverride {z=-200}
    physicsBodyDefinitions [{
      sphereCollisionRadius=0
      boundBone {boneName=dyn_lock_09}
      boxExtents=<10 15 15>
      constraintSetup {angularTargetAxis=AxisY, angularLimitsMax=<0 0 5>, angularLimitsMin=<0 -10 -5>, angularTarget=<0 0 0>}
      localJointOffset=<3 0 0>
     }]
   }
   n12 = Anim Dynamics - Bone: dyn_lock_10(ComponentPose=@n11, Alpha=1, ExternalForce="0,0,0") {
    angularDampingOverride=0.75
    angularSpringConstant=1500
    bOverrideAngularDamping=true
    bUseGravityOverride=true
    gravityScale=0.35
    simulationSpace=RootRelative
    alphaBoolBlend {blendOption=Linear}
    boundBone {boneName=dyn_lock_10}
    componentAppliedLinearAccClamp=<200 200 200>
    componentLinearAccScale=<0.2 0.2 0.2>
    componentLinearVelScale=<0.2 0.2 0.2>
    gravityOverride {z=-200}
    physicsBodyDefinitions [{
      sphereCollisionRadius=0
      boundBone {boneName=dyn_lock_10}
      boxExtents=<10 15 15>
      constraintSetup {angularTargetAxis=AxisY, angularLimitsMax=<0 0 5>, angularLimitsMin=<0 -10 -5>, angularTarget=<0 0 0>}
      localJointOffset=<3 0 0>
     }]
   }
   Output Pose(Result=Component To Local(ComponentPose=@n12)) {layerGroup=None}
   Save cached pose 'InputCachePose'(Pose=Copy Pose From Mesh() {bUseAttachedParent=true}) {cacheName=InputCachePose}
   n13 = "Transform (Modify) Bone - Bone: FACIAL_C_Jaw"!(
    ComponentPose=Local To Component(LocalPose=Use cached pose 'UpdatedBodyBones'())
    bAlphaBoolEnabled:bool=True
    Translation="0,0,0"
    Rotation="0,0,20"
    Scale="1,1,1"
   ) -> (Pose) {
    alphaInputType=Bool
    rotationMode=BMM_Replace
    rotationSpace=BCS_ParentBoneSpace
    translationMode=BMM_Replace
    translationSpace=BCS_ParentBoneSpace
    alphaBoolBlend {blendOption=Linear}
    boneToModify {boneName=FACIAL_C_Jaw}
    rotation {roll=20}
   }
   n14 = "Transform (Modify) Bone - Bone: FACIAL_C_Jawline"!(ComponentPose=@n13, bAlphaBoolEnabled:bool=True, Translation="0,0,0", Rotation="0,0,20", Scale="1,1,1") -> (Pose) {
    alphaInputType=Bool
    rotationMode=BMM_Replace
    rotationSpace=BCS_ParentBoneSpace
    translationMode=BMM_Replace
    translationSpace=BCS_ParentBoneSpace
    alphaBoolBlend {blendOption=Linear}
    boneToModify {boneName=FACIAL_C_Jawline}
    rotation {roll=20}
   }
   n15 = "Transform (Modify) Bone - Bone: FACIAL_L_Jawline"!(ComponentPose=@n14, bAlphaBoolEnabled:bool=True, Translation="0,0,0", Rotation="0,0,20", Scale="1,1,1") -> (Pose) {
    alphaInputType=Bool
    rotationMode=BMM_Replace
    rotationSpace=BCS_ParentBoneSpace
    translationMode=BMM_Replace
    translationSpace=BCS_ParentBoneSpace
    alphaBoolBlend {blendOption=Linear}
    boneToModify {boneName=FACIAL_L_Jawline}
    rotation {roll=20}
   }
   n16 = "Transform (Modify) Bone - Bone: FACIAL_R_Jawline"!(ComponentPose=@n15, bAlphaBoolEnabled:bool=True, Translation="0,0,0", Rotation="0,0,20", Scale="1,1,1") -> (Pose) {
    alphaInputType=Bool
    rotationMode=BMM_Replace
    rotationSpace=BCS_ParentBoneSpace
    translationMode=BMM_Replace
    translationSpace=BCS_ParentBoneSpace
    alphaBoolBlend {blendOption=Linear}
    boneToModify {boneName=FACIAL_R_Jawline}
    rotation {roll=20}
   }
   n17 = "Transform (Modify) Bone - Bone: FACIAL_C_Chin"!(ComponentPose=@n16, bAlphaBoolEnabled:bool=True, Translation="0,0,0", Rotation="0,0,20", Scale="1,1,1") -> (Pose) {
    alphaInputType=Bool
    rotationMode=BMM_Replace
    rotationSpace=BCS_ParentBoneSpace
    translationMode=BMM_Replace
    translationSpace=BCS_ParentBoneSpace
    alphaBoolBlend {blendOption=Linear}
    boneToModify {boneName=FACIAL_C_Chin}
    rotation {roll=20}
   }
   n18 = "Transform (Modify) Bone - Bone: FACIAL_L_ChinSide"!(ComponentPose=@n17, bAlphaBoolEnabled:bool=True, Translation="0,0,0", Rotation="0,0,20", Scale="1,1,1") -> (Pose) {
    alphaInputType=Bool
    rotationMode=BMM_Replace
    rotationSpace=BCS_ParentBoneSpace
    translationMode=BMM_Replace
    translationSpace=BCS_ParentBoneSpace
    alphaBoolBlend {blendOption=Linear}
    boneToModify {boneName=FACIAL_L_ChinSide}
    rotation {roll=20}
   }
   n19 = "Transform (Modify) Bone - Bone: FACIAL_R_ChinSide"!(ComponentPose=@n18, bAlphaBoolEnabled:bool=True, Translation="0,0,0", Rotation="0,0,20", Scale="1,1,1") -> (Pose) {
    alphaInputType=Bool
    rotationMode=BMM_Replace
    rotationSpace=BCS_ParentBoneSpace
    translationMode=BMM_Replace
    translationSpace=BCS_ParentBoneSpace
    alphaBoolBlend {blendOption=Linear}
    boneToModify {boneName=FACIAL_R_ChinSide}
    rotation {roll=20}
   }
   n20 = "Transform (Modify) Bone - Bone: FACIAL_C_UnderChin"!(ComponentPose=@n19, bAlphaBoolEnabled:bool=True, Translation="0,0,0", Rotation="0,0,20", Scale="1,1,1") -> (Pose) {
    alphaInputType=Bool
    rotationMode=BMM_Replace
    rotationSpace=BCS_ParentBoneSpace
    translationMode=BMM_Replace
    translationSpace=BCS_ParentBoneSpace
    alphaBoolBlend {blendOption=Linear}
    boneToModify {boneName=FACIAL_C_UnderChin}
    rotation {roll=20}
   }
   n21 = "Transform (Modify) Bone - Bone: FACIAL_L_UnderChin"!(ComponentPose=@n20, bAlphaBoolEnabled:bool=True, Translation="0,0,0", Rotation="0,0,20", Scale="1,1,1") -> (Pose) {
    alphaInputType=Bool
    rotationMode=BMM_Replace
    rotationSpace=BCS_ParentBoneSpace
    translationMode=BMM_Replace
    translationSpace=BCS_ParentBoneSpace
    alphaBoolBlend {blendOption=Linear}
    boneToModify {boneName=FACIAL_L_UnderChin}
    rotation {roll=20}
   }
   n22 = "Transform (Modify) Bone - Bone: FACIAL_R_UnderChin"!(ComponentPose=@n21, bAlphaBoolEnabled:bool=True, Translation="0,0,0", Rotation="0,0,20", Scale="1,1,1") -> (Pose) {
    alphaInputType=Bool
    rotationMode=BMM_Replace
    rotationSpace=BCS_ParentBoneSpace
    translationMode=BMM_Replace
    translationSpace=BCS_ParentBoneSpace
    alphaBoolBlend {blendOption=Linear}
    boneToModify {boneName=FACIAL_R_UnderChin}
    rotation {roll=20}
   }
   n23 = "Transform (Modify) Bone - Bone: FACIAL_L_CheekInner"!(ComponentPose=@n22, bAlphaBoolEnabled:bool=True, Translation="0,0,0", Rotation="0,0,20", Scale="1,1,1") -> (Pose) {
    alphaInputType=Bool
    rotationMode=BMM_Replace
    rotationSpace=BCS_ParentBoneSpace
    translationMode=BMM_Replace
    translationSpace=BCS_ParentBoneSpace
    alphaBoolBlend {blendOption=Linear}
    boneToModify {boneName=FACIAL_L_CheekInner}
    rotation {roll=20}
   }
   n24 = "Transform (Modify) Bone - Bone: FACIAL_R_CheekInner"!(ComponentPose=@n23, bAlphaBoolEnabled:bool=True, Translation="0,0,0", Rotation="0,0,20", Scale="1,1,1") -> (Pose) {
    alphaInputType=Bool
    rotationMode=BMM_Replace
    rotationSpace=BCS_ParentBoneSpace
    translationMode=BMM_Replace
    translationSpace=BCS_ParentBoneSpace
    alphaBoolBlend {blendOption=Linear}
    boneToModify {boneName=FACIAL_R_CheekInner}
    rotation {roll=20}
   }
   n25 = "Transform (Modify) Bone - Bone: FACIAL_L_CheekLower"!(ComponentPose=@n24, bAlphaBoolEnabled:bool=True, Translation="0,0,0", Rotation="0,0,20", Scale="1,1,1") -> (Pose) {
    alphaInputType=Bool
    rotationMode=BMM_Replace
    rotationSpace=BCS_ParentBoneSpace
    translationMode=BMM_Replace
    translationSpace=BCS_ParentBoneSpace
    alphaBoolBlend {blendOption=Linear}
    boneToModify {boneName=FACIAL_L_CheekLower}
    rotation {roll=20}
   }
   n26 = "Transform (Modify) Bone - Bone: FACIAL_R_CheekLower"!(ComponentPose=@n25, bAlphaBoolEnabled:bool=True, Translation="0,0,0", Rotation="0,0,20", Scale="1,1,1") -> (Pose) {
    alphaInputType=Bool
    rotationMode=BMM_Replace
    rotationSpace=BCS_ParentBoneSpace
    translationMode=BMM_Replace
    translationSpace=BCS_ParentBoneSpace
    alphaBoolBlend {blendOption=Linear}
    boneToModify {boneName=FACIAL_R_CheekLower}
    rotation {roll=20}
   }
   n27 = "Transform (Modify) Bone - Bone: FACIAL_L_CheekOuter"!(ComponentPose=@n26, bAlphaBoolEnabled:bool=True, Translation="0,0,0", Rotation="0,0,20", Scale="1,1,1") -> (Pose) {
    alphaInputType=Bool
    rotationMode=BMM_Replace
    rotationSpace=BCS_ParentBoneSpace
    translationMode=BMM_Replace
    translationSpace=BCS_ParentBoneSpace
    alphaBoolBlend {blendOption=Linear}
    boneToModify {boneName=FACIAL_L_CheekOuter}
    rotation {roll=20}
   }
   n28 = "Transform (Modify) Bone - Bone: FACIAL_R_CheekOuter"!(ComponentPose=@n27, bAlphaBoolEnabled:bool=True, Translation="0,0,0", Rotation="0,0,20", Scale="1,1,1") -> (Pose) {
    alphaInputType=Bool
    rotationMode=BMM_Replace
    rotationSpace=BCS_ParentBoneSpace
    translationMode=BMM_Replace
    translationSpace=BCS_ParentBoneSpace
    alphaBoolBlend {blendOption=Linear}
    boneToModify {boneName=FACIAL_R_CheekOuter}
    rotation {roll=20}
   }
   n29 = "Transform (Modify) Bone - Bone: FACIAL_C_MouthUpper"!(ComponentPose=@n28, bAlphaBoolEnabled:bool=True, Translation="0,0,0", Rotation="0,0,20", Scale="1,1,1") -> (Pose) {
    alphaInputType=Bool
    rotationMode=BMM_Replace
    rotationSpace=BCS_ParentBoneSpace
    translationMode=BMM_Replace
    translationSpace=BCS_ParentBoneSpace
    alphaBoolBlend {blendOption=Linear}
    boneToModify {boneName=FACIAL_C_MouthUpper}
    rotation {roll=20}
   }
   n30 = "Transform (Modify) Bone - Bone: FACIAL_L_LipCorner"!(ComponentPose=@n29, bAlphaBoolEnabled:bool=True, Translation="0,0,0", Rotation="0,0,20", Scale="1,1,1") -> (Pose) {
    alphaInputType=Bool
    rotationMode=BMM_Replace
    rotationSpace=BCS_ParentBoneSpace
    translationMode=BMM_Replace
    translationSpace=BCS_ParentBoneSpace
    alphaBoolBlend {blendOption=Linear}
    boneToModify {boneName=FACIAL_L_LipCorner}
    rotation {roll=20}
   }
   n31 = "Transform (Modify) Bone - Bone: FACIAL_R_LipCorner"!(ComponentPose=@n30, bAlphaBoolEnabled:bool=True, Translation="0,0,0", Rotation="0,0,20", Scale="1,1,1") -> (Pose) {
    alphaInputType=Bool
    rotationMode=BMM_Replace
    rotationSpace=BCS_ParentBoneSpace
    translationMode=BMM_Replace
    translationSpace=BCS_ParentBoneSpace
    alphaBoolBlend {blendOption=Linear}
    boneToModify {boneName=FACIAL_R_LipCorner}
    rotation {roll=20}
   }
   n32 = "Transform (Modify) Bone - Bone: FACIAL_C_LowerLipRotation"!(ComponentPose=@n31, bAlphaBoolEnabled:bool=True, Translation="0,0,0", Rotation="0,0,20", Scale="1,1,1") -> (Pose) {
    alphaInputType=Bool
    rotationMode=BMM_Replace
    rotationSpace=BCS_ParentBoneSpace
    translationMode=BMM_Replace
    translationSpace=BCS_ParentBoneSpace
    alphaBoolBlend {blendOption=Linear}
    boneToModify {boneName=FACIAL_C_LowerLipRotation}
    rotation {roll=20}
   }
   n33 = "Transform (Modify) Bone - Bone: FACIAL_C_MouthLower"!(ComponentPose=@n32, bAlphaBoolEnabled:bool=True, Translation="0,0,0", Rotation="0,0,20", Scale="1,1,1") -> (Pose) {
    alphaInputType=Bool
    rotationMode=BMM_Replace
    rotationSpace=BCS_ParentBoneSpace
    translationMode=BMM_Replace
    translationSpace=BCS_ParentBoneSpace
    alphaBoolBlend {blendOption=Linear}
    boneToModify {boneName=FACIAL_C_MouthLower}
    rotation {roll=20}
   }
   n34 = "Transform (Modify) Bone - Bone: FACIAL_L_LipLowerOuter"!(ComponentPose=@n33, bAlphaBoolEnabled:bool=True, Translation="0,0,0", Rotation="0,0,20", Scale="1,1,1") -> (Pose) {
    alphaInputType=Bool
    rotationMode=BMM_Replace
    rotationSpace=BCS_ParentBoneSpace
    translationMode=BMM_Replace
    translationSpace=BCS_ParentBoneSpace
    alphaBoolBlend {blendOption=Linear}
    boneToModify {boneName=FACIAL_L_LipLowerOuter}
    rotation {roll=20}
   }
   n35 = "Transform (Modify) Bone - Bone: FACIAL_R_LipLowerOuter"!(ComponentPose=@n34, bAlphaBoolEnabled:bool=True, Translation="0,0,0", Rotation="0,0,20", Scale="1,1,1") -> (Pose) {
    alphaInputType=Bool
    rotationMode=BMM_Replace
    rotationSpace=BCS_ParentBoneSpace
    translationMode=BMM_Replace
    translationSpace=BCS_ParentBoneSpace
    alphaBoolBlend {blendOption=Linear}
    boneToModify {boneName=FACIAL_R_LipLowerOuter}
    rotation {roll=20}
   }
   n36 = "Transform (Modify) Bone - Bone: FACIAL_L_NasolabialBulge"!(ComponentPose=@n35, bAlphaBoolEnabled:bool=True, Translation="0,0,0", Rotation="0,0,20", Scale="1,1,1") -> (Pose) {
    alphaInputType=Bool
    rotationMode=BMM_Replace
    rotationSpace=BCS_ParentBoneSpace
    translationMode=BMM_Replace
    translationSpace=BCS_ParentBoneSpace
    alphaBoolBlend {blendOption=Linear}
    boneToModify {boneName=FACIAL_L_NasolabialBulge}
    rotation {roll=20}
   }
   n37 = "Transform (Modify) Bone - Bone: FACIAL_R_NasolabialBulge"!(ComponentPose=@n36, bAlphaBoolEnabled:bool=True, Translation="0,0,0", Rotation="0,0,20", Scale="1,1,1") -> (Pose) {
    alphaInputType=Bool
    rotationMode=BMM_Replace
    rotationSpace=BCS_ParentBoneSpace
    translationMode=BMM_Replace
    translationSpace=BCS_ParentBoneSpace
    alphaBoolBlend {blendOption=Linear}
    boneToModify {boneName=FACIAL_R_NasolabialBulge}
    rotation {roll=20}
   }
   n38 = "Transform (Modify) Bone - Bone: FACIAL_L_NasolabialFurrow"!(ComponentPose=@n37, bAlphaBoolEnabled:bool=True, Translation="0,0,0", Rotation="0,0,20", Scale="1,1,1") -> (Pose) {
    alphaInputType=Bool
    rotationMode=BMM_Replace
    rotationSpace=BCS_ParentBoneSpace
    translationMode=BMM_Replace
    translationSpace=BCS_ParentBoneSpace
    alphaBoolBlend {blendOption=Linear}
    boneToModify {boneName=FACIAL_L_NasolabialFurrow}
    rotation {roll=20}
   }
   n39 = "Transform (Modify) Bone - Bone: FACIAL_R_NasolabialFurrow"!(ComponentPose=@n38, bAlphaBoolEnabled:bool=True, Translation="0,0,0", Rotation="0,0,20", Scale="1,1,1") -> (Pose) {
    alphaInputType=Bool
    rotationMode=BMM_Replace
    rotationSpace=BCS_ParentBoneSpace
    translationMode=BMM_Replace
    translationSpace=BCS_ParentBoneSpace
    alphaBoolBlend {blendOption=Linear}
    boneToModify {boneName=FACIAL_R_NasolabialFurrow}
    rotation {roll=20}
   }
   n40 = "Transform (Modify) Bone - Bone: FACIAL_L_Masseter"!(ComponentPose=@n39, bAlphaBoolEnabled:bool=True, Translation="0,0,0", Rotation="0,0,20", Scale="1,1,1") -> (Pose) {
    alphaInputType=Bool
    rotationMode=BMM_Replace
    rotationSpace=BCS_ParentBoneSpace
    translationMode=BMM_Replace
    translationSpace=BCS_ParentBoneSpace
    alphaBoolBlend {blendOption=Linear}
    boneToModify {boneName=FACIAL_L_Masseter}
    rotation {roll=20}
   }
   n41 = "Transform (Modify) Bone - Bone: FACIAL_R_Masseter"!(ComponentPose=@n40, bAlphaBoolEnabled:bool=True, Translation="0,0,0", Rotation="0,0,20", Scale="1,1,1") -> (Pose) {
    alphaInputType=Bool
    rotationMode=BMM_Replace
    rotationSpace=BCS_ParentBoneSpace
    translationMode=BMM_Replace
    translationSpace=BCS_ParentBoneSpace
    alphaBoolBlend {blendOption=Linear}
    boneToModify {boneName=FACIAL_R_Masseter}
    rotation {roll=20}
   }
   n42 = "Transform (Modify) Bone - Bone: FACIAL_L_Ear"!(ComponentPose=@n41, bAlphaBoolEnabled:bool=True, Translation="0,0,0", Rotation="0,0,20", Scale="1,1,1") -> (Pose) {
    alphaInputType=Bool
    rotationMode=BMM_Replace
    rotationSpace=BCS_ParentBoneSpace
    translationMode=BMM_Replace
    translationSpace=BCS_ParentBoneSpace
    alphaBoolBlend {blendOption=Linear}
    boneToModify {boneName=FACIAL_L_Ear}
    rotation {roll=20}
   }
   n43 = "Transform (Modify) Bone - Bone: FACIAL_R_Ear"!(ComponentPose=@n42, bAlphaBoolEnabled:bool=True, Translation="0,0,0", Rotation="0,0,20", Scale="1,1,1") -> (Pose) {
    alphaInputType=Bool
    rotationMode=BMM_Replace
    rotationSpace=BCS_ParentBoneSpace
    translationMode=BMM_Replace
    translationSpace=BCS_ParentBoneSpace
    alphaBoolBlend {blendOption=Linear}
    boneToModify {boneName=FACIAL_R_Ear}
    rotation {roll=20}
   }
   n44 = "Transform (Modify) Bone - Bone: FACIAL_L_EyeCornerOuter"!(ComponentPose=@n43, bAlphaBoolEnabled:bool=True, Translation="0,0,0", Rotation="0,0,20", Scale="1,1,1") -> (Pose) {
    alphaInputType=Bool
    rotationMode=BMM_Replace
    rotationSpace=BCS_ParentBoneSpace
    translationMode=BMM_Replace
    translationSpace=BCS_ParentBoneSpace
    alphaBoolBlend {blendOption=Linear}
    boneToModify {boneName=FACIAL_L_EyeCornerOuter}
    rotation {roll=20}
   }
   n45 = "Transform (Modify) Bone - Bone: FACIAL_R_EyeCornerOuter"!(ComponentPose=@n44, bAlphaBoolEnabled:bool=True, Translation="0,0,0", Rotation="0,0,20", Scale="1,1,1") -> (Pose) {
    alphaInputType=Bool
    rotationMode=BMM_Replace
    rotationSpace=BCS_ParentBoneSpace
    translationMode=BMM_Replace
    translationSpace=BCS_ParentBoneSpace
    alphaBoolBlend {blendOption=Linear}
    boneToModify {boneName=FACIAL_R_EyeCornerOuter}
    rotation {roll=20}
   }
   n46 = "Transform (Modify) Bone - Bone: FACIAL_L_EyesackLower"!(ComponentPose=@n45, bAlphaBoolEnabled:bool=True, Translation="0,0,0", Rotation="0,0,20", Scale="1,1,1") -> (Pose) {
    alphaInputType=Bool
    rotationMode=BMM_Replace
    rotationSpace=BCS_ParentBoneSpace
    translationMode=BMM_Replace
    translationSpace=BCS_ParentBoneSpace
    alphaBoolBlend {blendOption=Linear}
    boneToModify {boneName=FACIAL_L_EyesackLower}
    rotation {roll=20}
   }
   n47 = "Transform (Modify) Bone - Bone: FACIAL_R_EyesackLower"!(ComponentPose=@n46, bAlphaBoolEnabled:bool=True, Translation="0,0,0", Rotation="0,0,20", Scale="1,1,1") -> (Pose) {
    alphaInputType=Bool
    rotationMode=BMM_Replace
    rotationSpace=BCS_ParentBoneSpace
    translationMode=BMM_Replace
    translationSpace=BCS_ParentBoneSpace
    alphaBoolBlend {blendOption=Linear}
    boneToModify {boneName=FACIAL_R_EyesackLower}
    rotation {roll=20}
   }
   Save cached pose 'UpdatedFaceBones'(Pose=Component To Local(ComponentPose=@n47)) {cacheName=UpdatedFaceBones}
   Save cached pose 'PostFaceUpdate'(Pose=Blend Poses by bool(
     BlendPose_0=Use cached pose 'UpdatedFaceBones'()
     BlendPose_1=Use cached pose 'UpdatedBodyBones'()
     bActiveValue=Get bHasFaceBones() -> (bHasFaceBones:bool)
    ) {blendType=Linear}) {cacheName=PostFaceUpdate}
   n48 = "Transform (Modify) Bone - Bone: upperarm_l"!(
    ComponentPose=Local To Component(LocalPose=Use cached pose 'InputCachePose'())
    bAlphaBoolEnabled:bool=True
    Translation="0,0,0"
    Rotation="0,0,20"
    Scale="1,1,1"
   ) -> (Pose) {
    alphaInputType=Bool
    rotationMode=BMM_Replace
    rotationSpace=BCS_ParentBoneSpace
    translationMode=BMM_Replace
    translationSpace=BCS_ParentBoneSpace
    alphaBoolBlend {blendOption=Linear}
    boneToModify {boneName=upperarm_l}
    rotation {roll=20}
   }
   n49 = "Transform (Modify) Bone - Bone: upperarm_twist_01_l"!(ComponentPose=@n48, bAlphaBoolEnabled:bool=True, Translation="0,0,0", Rotation="0,0,20", Scale="1,1,1") -> (Pose) {
    alphaInputType=Bool
    rotationMode=BMM_Replace
    rotationSpace=BCS_ParentBoneSpace
    translationMode=BMM_Replace
    translationSpace=BCS_ParentBoneSpace
    alphaBoolBlend {blendOption=Linear}
    boneToModify {boneName=upperarm_twist_01_l}
    rotation {roll=20}
   }
   n50 = "Transform (Modify) Bone - Bone: deform_upperarm_up_base_l"!(ComponentPose=@n49, bAlphaBoolEnabled:bool=True, Translation="0,0,0", Rotation="0,0,20", Scale="1,1,1") -> (Pose) {
    alphaInputType=Bool
    rotationMode=BMM_Replace
    rotationSpace=BCS_ParentBoneSpace
    translationMode=BMM_Replace
    translationSpace=BCS_ParentBoneSpace
    alphaBoolBlend {blendOption=Linear}
    boneToModify {boneName=deform_upperarm_up_base_l}
    rotation {roll=20}
   }
   n51 = "Transform (Modify) Bone - Bone: deform_upperarm_up_l"!(ComponentPose=@n50, bAlphaBoolEnabled:bool=True, Translation="0,0,0", Rotation="0,0,20", Scale="1,1,1") -> (Pose) {
    alphaInputType=Bool
    rotationMode=BMM_Replace
    rotationSpace=BCS_ParentBoneSpace
    translationMode=BMM_Replace
    translationSpace=BCS_ParentBoneSpace
    alphaBoolBlend {blendOption=Linear}
    boneToModify {boneName=deform_upperarm_up_l}
    rotation {roll=20}
   }
   n52 = "Transform (Modify) Bone - Bone: upperarm_r"!(ComponentPose=@n51, bAlphaBoolEnabled:bool=True, Translation="0,0,0", Rotation="0,0,20", Scale="1,1,1") -> (Pose) {
    alphaInputType=Bool
    rotationMode=BMM_Replace
    rotationSpace=BCS_ParentBoneSpace
    translationMode=BMM_Replace
    translationSpace=BCS_ParentBoneSpace
    alphaBoolBlend {blendOption=Linear}
    boneToModify {boneName=upperarm_r}
    rotation {roll=20}
   }
   n53 = "Transform (Modify) Bone - Bone: upperarm_twist_01_r"!(ComponentPose=@n52, bAlphaBoolEnabled:bool=True, Translation="0,0,0", Rotation="0,0,20", Scale="1,1,1") -> (Pose) {
    alphaInputType=Bool
    rotationMode=BMM_Replace
    rotationSpace=BCS_ParentBoneSpace
    translationMode=BMM_Replace
    translationSpace=BCS_ParentBoneSpace
    alphaBoolBlend {blendOption=Linear}
    boneToModify {boneName=upperarm_twist_01_r}
    rotation {roll=20}
   }
   n54 = "Transform (Modify) Bone - Bone: deform_upperarm_up_base_r"!(ComponentPose=@n53, bAlphaBoolEnabled:bool=True, Translation="0,0,0", Rotation="0,0,20", Scale="1,1,1") -> (Pose) {
    alphaInputType=Bool
    rotationMode=BMM_Replace
    rotationSpace=BCS_ParentBoneSpace
    translationMode=BMM_Replace
    translationSpace=BCS_ParentBoneSpace
    alphaBoolBlend {blendOption=Linear}
    boneToModify {boneName=deform_upperarm_up_base_r}
    rotation {roll=20}
   }
   n55 = "Transform (Modify) Bone - Bone: deform_upperarm_up_r"!(ComponentPose=@n54, bAlphaBoolEnabled:bool=True, Translation="0,0,0", Rotation="0,0,20", Scale="1,1,1") -> (Pose) {
    alphaInputType=Bool
    rotationMode=BMM_Replace
    rotationSpace=BCS_ParentBoneSpace
    translationMode=BMM_Replace
    translationSpace=BCS_ParentBoneSpace
    alphaBoolBlend {blendOption=Linear}
    boneToModify {boneName=deform_upperarm_up_r}
    rotation {roll=20}
   }
   Save cached pose 'UpdatedBodyBones'(Pose=Component To Local(ComponentPose=@n55)) {cacheName=UpdatedBodyBones}
  ]}
 EventGraph {nodes [
   Event BlueprintUpdateAnimation()
   > n1 = Gravity Override State Function DEPRECATED(
    CharacterPartAnimInstance=Self-Reference()
    GameTimeInSeconds=GetGameTimeInSeconds()
    GravityOverrideParams=Get Gravity Override Params()
    RBAN Params=Get RBAN Params()
   )
   > Set Gravity Override(Gravity Override=@n1)
   > n2 = Sequence()
   > Is Valid(InputObject=Get HeadMesh() -> (HeadMesh:Skeletal Mesh Component Object Reference))
   n3 = GetSocketBoneName!(self:Skinned Mesh Component Object Reference=Get HeadMesh(), InSocketName:Name=FACIAL_C_FacialRoot) -> (ReturnValue:Name=None)
   > n4 = Set bHasFaceBones(bHasFaceBones:bool="Not Equal Exactly (String)"(A="To String (Name)"(InName=@n3), B=None)) -> (Output_Get:bool)
   > Branch(Condition=@n4)
   > K2Node_Composite_1 = UpdateHeadBoneValues() {is=title K2Node_Composite}
   Is Valid(exec=@n2.then_1, InputObject=Get BodyMesh() -> (BodyMesh:Skeletal Mesh Component Budgeted Object Reference))
   > K2Node_Composite_0 = UpdateBodyBonesValues() {is=title K2Node_Composite}
   n5 = Event OnCharacterCustomizationLoaded()
   > Set HeadMesh(HeadMesh:Skeletal Mesh Component Object Reference=Get Array Item with EFort Custom Part Type(Array=@n5.CustomMeshParts, PartType=Head)) -> (Output_Get:Skeletal Mesh Component Object Reference)
   > Set BodyMesh(BodyMesh:Skeletal Mesh Component Budgeted Object Reference=Get Array Item with EFort Custom Part Type(Array=@n5.CustomMeshParts)) -> (Output_Get:Skeletal Mesh Component Budgeted Object Reference)
  ]}
 UpdateBodyBonesValues {
  path=EventGraph.K2Node_Composite_0.UpdateBodyBonesValues
  nodes [
   Inputs()
   n1 = BreakTransform(InTransform=GetSocketTransform(self=Get BodyMesh(), InSocketName=upperarm_l, TransformSpace=RTS_ParentBoneSpace))
   > Set upperarm_l_Location(upperarm_l_Location:Vector=@n1) -> (Output_Get:Vector)
   > Set upperarm_l_Rotation(upperarm_l_Rotation:Rotator=@n1.Rotation) -> (Output_Get:Rotator)
   n2 = BreakTransform(InTransform=GetSocketTransform(self=Get BodyMesh(), InSocketName=upperarm_r, TransformSpace=RTS_ParentBoneSpace))
   > Set upperarm_r_Location(upperarm_r_Location:Vector=@n2) -> (Output_Get:Vector)
   > Set upperarm_r_Rotation(upperarm_r_Rotation:Rotator=@n2.Rotation) -> (Output_Get:Rotator)
   n3 = BreakTransform(InTransform=GetSocketTransform(self=Get BodyMesh(), InSocketName=upperarm_twist_01_l, TransformSpace=RTS_ParentBoneSpace))
   > Set upperarm_twist_01_l_Location(upperarm_twist_01_l_Location:Vector=@n3) -> (Output_Get:Vector)
   > Set upperarm_twist_01_l_Rotation(upperarm_twist_01_l_Rotation:Rotator=@n3.Rotation) -> (Output_Get:Rotator)
   n4 = BreakTransform(InTransform=GetSocketTransform(self=Get BodyMesh(), InSocketName=upperarm_twist_01_r, TransformSpace=RTS_ParentBoneSpace))
   > Set upperarm_twist_01_r_Location(upperarm_twist_01_r_Location:Vector=@n4) -> (Output_Get:Vector)
   > Set upperarm_twist_01_r_Rotation(upperarm_twist_01_r_Rotation:Rotator=@n4.Rotation) -> (Output_Get:Rotator)
   n5 = BreakTransform(InTransform=GetSocketTransform(self=Get BodyMesh(), InSocketName=deform_upperarm_up_base_l, TransformSpace=RTS_ParentBoneSpace))
   > Set deform_upperarm_up_base_l_Location(deform_upperarm_up_base_l_Location:Vector=@n5) -> (Output_Get:Vector)
   > Set deform_upperarm_up_base_l_Rotation(deform_upperarm_up_base_l_Rotation:Rotator=@n5.Rotation) -> (Output_Get:Rotator)
   n6 = BreakTransform(InTransform=GetSocketTransform(self=Get BodyMesh(), InSocketName=deform_upperarm_up_base_r, TransformSpace=RTS_ParentBoneSpace))
   > Set deform_upperarm_up_base_r_Location(deform_upperarm_up_base_r_Location:Vector=@n6) -> (Output_Get:Vector)
   > Set deform_upperarm_up_base_r_Rotation(deform_upperarm_up_base_r_Rotation:Rotator=@n6.Rotation) -> (Output_Get:Rotator)
   n7 = BreakTransform(InTransform=GetSocketTransform(self=Get BodyMesh(), InSocketName=deform_upperarm_up_l, TransformSpace=RTS_ParentBoneSpace))
   > Set deform_upperarm_up_l_Location(deform_upperarm_up_l_Location:Vector=@n7) -> (Output_Get:Vector)
   > Set deform_upperarm_up_l_Rotation(deform_upperarm_up_l_Rotation:Rotator=@n7.Rotation) -> (Output_Get:Rotator)
   n8 = BreakTransform(InTransform=GetSocketTransform(self=Get BodyMesh(), InSocketName=deform_upperarm_up_r, TransformSpace=RTS_ParentBoneSpace))
   > Set deform_upperarm_up_r_Location(deform_upperarm_up_r_Location:Vector=@n8) -> (Output_Get:Vector)
   > Set deform_upperarm_up_r_Rotation(deform_upperarm_up_r_Rotation:Rotator=@n8.Rotation) -> (Output_Get:Rotator)
   Outputs()
  ]
 }
 UpdateHeadBoneValues {
  path=EventGraph.K2Node_Composite_1.UpdateHeadBoneValues
  nodes [
   Inputs()
   n1 = BreakTransform(InTransform=GetSocketTransform(self=Get HeadMesh(), InSocketName=FACIAL_C_Jaw, TransformSpace=RTS_ParentBoneSpace))
   > Set FACIAL_C_Jaw_Location(FACIAL_C_Jaw_Location:Vector=@n1) -> (Output_Get:Vector)
   > Set FACIAL_C_Jaw_Rotation(FACIAL_C_Jaw_Rotation:Rotator=@n1.Rotation) -> (Output_Get:Rotator)
   n2 = BreakTransform(InTransform=GetSocketTransform(self=Get HeadMesh(), InSocketName=FACIAL_C_Chin, TransformSpace=RTS_ParentBoneSpace))
   > Set FACIAL_C_Chin_Location(FACIAL_C_Chin_Location:Vector=@n2) -> (Output_Get:Vector)
   > Set FACIAL_C_Chin_Rotation(FACIAL_C_Chin_Rotation:Rotator=@n2.Rotation) -> (Output_Get:Rotator)
   n3 = BreakTransform(InTransform=GetSocketTransform(self=Get HeadMesh(), InSocketName=FACIAL_C_Jawline, TransformSpace=RTS_ParentBoneSpace))
   > Set FACIAL_C_Jawline_Location(FACIAL_C_Jawline_Location:Vector=@n3) -> (Output_Get:Vector)
   > Set FACIAL_C_Jawline_Rotation(FACIAL_C_Jawline_Rotation:Rotator=@n3.Rotation) -> (Output_Get:Rotator)
   n4 = BreakTransform(InTransform=GetSocketTransform(self=Get HeadMesh(), InSocketName=FACIAL_L_Jawline, TransformSpace=RTS_ParentBoneSpace))
   > Set FACIAL_L_Jawline_Location(FACIAL_L_Jawline_Location:Vector=@n4) -> (Output_Get:Vector)
   > Set FACIAL_L_Jawline_Rotation(FACIAL_L_Jawline_Rotation:Rotator=@n4.Rotation) -> (Output_Get:Rotator)
   n5 = BreakTransform(InTransform=GetSocketTransform(self=Get HeadMesh(), InSocketName=FACIAL_R_Jawline, TransformSpace=RTS_ParentBoneSpace))
   > Set FACIAL_R_Jawline_Location(FACIAL_R_Jawline_Location:Vector=@n5) -> (Output_Get:Vector)
   > Set FACIAL_R_Jawline_Rotation(FACIAL_R_Jawline_Rotation:Rotator=@n5.Rotation) -> (Output_Get:Rotator)
   n6 = BreakTransform(InTransform=GetSocketTransform(self=Get HeadMesh(), InSocketName=FACIAL_L_ChinSide, TransformSpace=RTS_ParentBoneSpace))
   > Set FACIAL_L_ChinSide_Location(FACIAL_L_ChinSide_Location:Vector=@n6) -> (Output_Get:Vector)
   > Set FACIAL_L_ChinSide_Rotation(FACIAL_L_ChinSide_Rotation:Rotator=@n6.Rotation) -> (Output_Get:Rotator)
   n7 = BreakTransform(InTransform=GetSocketTransform(self=Get HeadMesh(), InSocketName=FACIAL_R_ChinSide, TransformSpace=RTS_ParentBoneSpace))
   > Set FACIAL_R_ChinSide_Location(FACIAL_R_ChinSide_Location:Vector=@n7) -> (Output_Get:Vector)
   > Set FACIAL_R_ChinSide_Rotation(FACIAL_R_ChinSide_Rotation:Rotator=@n7.Rotation) -> (Output_Get:Rotator)
   n8 = BreakTransform(InTransform=GetSocketTransform(self=Get HeadMesh(), InSocketName=FACIAL_C_LowerLipRotation, TransformSpace=RTS_ParentBoneSpace))
   > Set FACIAL_C_LowerLipRotation_Location(FACIAL_C_LowerLipRotation_Location:Vector=@n8) -> (Output_Get:Vector)
   > Set FACIAL_C_LowerLipRotation_Rotation(FACIAL_C_LowerLipRotation_Rotation:Rotator=@n8.Rotation) -> (Output_Get:Rotator)
   n9 = BreakTransform(InTransform=GetSocketTransform(self=Get HeadMesh(), InSocketName=FACIAL_C_MouthLower, TransformSpace=RTS_ParentBoneSpace))
   > Set FACIAL_C_MouthLower_Location(FACIAL_C_MouthLower_Location:Vector=@n9) -> (Output_Get:Vector)
   > Set FACIAL_C_MouthLower_Rotation(FACIAL_C_MouthLower_Rotation:Rotator=@n9.Rotation) -> (Output_Get:Rotator)
   n10 = BreakTransform(InTransform=GetSocketTransform(self=Get HeadMesh(), InSocketName=FACIAL_L_LipLowerOuter, TransformSpace=RTS_ParentBoneSpace))
   > Set FACIAL_L_LipLowerOuter_Location(FACIAL_L_LipLowerOuter_Location:Vector=@n10) -> (Output_Get:Vector)
   > Set FACIAL_L_LipLowerOuter_Rotation(FACIAL_L_LipLowerOuter_Rotation:Rotator=@n10.Rotation) -> (Output_Get:Rotator)
   n11 = BreakTransform(InTransform=GetSocketTransform(self=Get HeadMesh(), InSocketName=FACIAL_R_LipLowerOuter, TransformSpace=RTS_ParentBoneSpace))
   > Set FACIAL_R_LipLowerOuter_Location(FACIAL_R_LipLowerOuter_Location:Vector=@n11) -> (Output_Get:Vector)
   > Set FACIAL_R_LipLowerOuter_Rotation(FACIAL_R_LipLowerOuter_Rotation:Rotator=@n11.Rotation) -> (Output_Get:Rotator)
   n12 = BreakTransform(InTransform=GetSocketTransform(self=Get HeadMesh(), InSocketName=FACIAL_C_MouthUpper, TransformSpace=RTS_ParentBoneSpace))
   > Set FACIAL_C_MouthUpper_Location(FACIAL_C_MouthUpper_Location:Vector=@n12) -> (Output_Get:Vector)
   > Set FACIAL_C_MouthUpper_Rotation(FACIAL_C_MouthUpper_Rotation:Rotator=@n12.Rotation) -> (Output_Get:Rotator)
   n13 = BreakTransform(InTransform=GetSocketTransform(self=Get HeadMesh(), InSocketName=FACIAL_L_LipCorner, TransformSpace=RTS_ParentBoneSpace))
   > Set FACIAL_L_LipCorner_Location(FACIAL_L_LipCorner_Location:Vector=@n13) -> (Output_Get:Vector)
   > Set FACIAL_L_LipCorner_Rotation(FACIAL_L_LipCorner_Rotation:Rotator=@n13.Rotation) -> (Output_Get:Rotator)
   n14 = BreakTransform(InTransform=GetSocketTransform(self=Get HeadMesh(), InSocketName=FACIAL_R_LipCorner, TransformSpace=RTS_ParentBoneSpace))
   > Set FACIAL_R_LipCorner_Location(FACIAL_R_LipCorner_Location:Vector=@n14) -> (Output_Get:Vector)
   > Set FACIAL_R_LipCorner_Rotation(FACIAL_R_LipCorner_Rotation:Rotator=@n14.Rotation) -> (Output_Get:Rotator)
   n15 = BreakTransform(InTransform=GetSocketTransform(self=Get HeadMesh(), InSocketName=FACIAL_C_UnderChin, TransformSpace=RTS_ParentBoneSpace))
   > Set FACIAL_C_UnderChin_Lccation(FACIAL_C_UnderChin_Lccation:Vector=@n15) -> (Output_Get:Vector)
   > Set FACIAL_C_UnderChin_Rotation(FACIAL_C_UnderChin_Rotation:Rotator=@n15.Rotation) -> (Output_Get:Rotator)
   n16 = BreakTransform(InTransform=GetSocketTransform(self=Get HeadMesh(), InSocketName=FACIAL_L_UnderChin, TransformSpace=RTS_ParentBoneSpace))
   > Set FACIAL_L_UnderChin_Location(FACIAL_L_UnderChin_Location:Vector=@n16) -> (Output_Get:Vector)
   > Set FACIAL_L_UnderChin_Rotation(FACIAL_L_UnderChin_Rotation:Rotator=@n16.Rotation) -> (Output_Get:Rotator)
   n17 = BreakTransform(InTransform=GetSocketTransform(self=Get HeadMesh(), InSocketName=FACIAL_R_UnderChin, TransformSpace=RTS_ParentBoneSpace))
   > Set FACIAL_R_UnderChin_Location(FACIAL_R_UnderChin_Location:Vector=@n17) -> (Output_Get:Vector)
   > Set FACIAL_R_UnderChin_Rotation(FACIAL_R_UnderChin_Rotation:Rotator=@n17.Rotation) -> (Output_Get:Rotator)
   n18 = BreakTransform(InTransform=GetSocketTransform(self=Get HeadMesh(), InSocketName=FACIAL_L_CheekInner, TransformSpace=RTS_ParentBoneSpace))
   > Set FACIAL_L_CheekInner_Location(FACIAL_L_CheekInner_Location:Vector=@n18) -> (Output_Get:Vector)
   > Set FACIAL_L_CheekInner_Rotation(FACIAL_L_CheekInner_Rotation:Rotator=@n18.Rotation) -> (Output_Get:Rotator)
   n19 = BreakTransform(InTransform=GetSocketTransform(self=Get HeadMesh(), InSocketName=FACIAL_R_CheekInner, TransformSpace=RTS_ParentBoneSpace))
   > Set FACIAL_R_CheekInner_Location(FACIAL_R_CheekInner_Location:Vector=@n19) -> (Output_Get:Vector)
   > Set FACIAL_R_CheekInner_Rotation(FACIAL_R_CheekInner_Rotation:Rotator=@n19.Rotation) -> (Output_Get:Rotator)
   n20 = BreakTransform(InTransform=GetSocketTransform(self=Get HeadMesh(), InSocketName=FACIAL_L_CheekLower, TransformSpace=RTS_ParentBoneSpace))
   > Set FACIAL_L_CheekLower_Location(FACIAL_L_CheekLower_Location:Vector=@n20) -> (Output_Get:Vector)
   > Set FACIAL_L_CheekLower_Rotation(FACIAL_L_CheekLower_Rotation:Rotator=@n20.Rotation) -> (Output_Get:Rotator)
   n21 = BreakTransform(InTransform=GetSocketTransform(self=Get HeadMesh(), InSocketName=FACIAL_R_CheekLower, TransformSpace=RTS_ParentBoneSpace))
   > Set FACIAL_R_CheekLower_Location(FACIAL_R_CheekLower_Location:Vector=@n21) -> (Output_Get:Vector)
   > Set FACIAL_R_CheekLower_Rotation(FACIAL_R_CheekLower_Rotation:Rotator=@n21.Rotation) -> (Output_Get:Rotator)
   n22 = BreakTransform(InTransform=GetSocketTransform(self=Get HeadMesh(), InSocketName=FACIAL_L_CheekOuter, TransformSpace=RTS_ParentBoneSpace))
   > Set FACIAL_L_CheekOuter_Location(FACIAL_L_CheekOuter_Location:Vector=@n22) -> (Output_Get:Vector)
   > Set FACIAL_L_CheekOuter_Rotation(FACIAL_L_CheekOuter_Rotation:Rotator=@n22.Rotation) -> (Output_Get:Rotator)
   n23 = BreakTransform(InTransform=GetSocketTransform(self=Get HeadMesh(), InSocketName=FACIAL_R_CheekOuter, TransformSpace=RTS_ParentBoneSpace))
   > Set FACIAL_R_CheekOuter_Location(FACIAL_R_CheekOuter_Location:Vector=@n23) -> (Output_Get:Vector)
   > Set FACIAL_R_CheekOuter_Rotation(FACIAL_R_CheekOuter_Rotation:Rotator=@n23.Rotation) -> (Output_Get:Rotator)
   n24 = BreakTransform(InTransform=GetSocketTransform(self=Get HeadMesh(), InSocketName=FACIAL_L_Ear, TransformSpace=RTS_ParentBoneSpace))
   > Set FACIAL_L_Ear_Location(FACIAL_L_Ear_Location:Vector=@n24) -> (Output_Get:Vector)
   > Set FACIAL_L_Ear_Rotation(FACIAL_L_Ear_Rotation:Rotator=@n24.Rotation) -> (Output_Get:Rotator)
   n25 = BreakTransform(InTransform=GetSocketTransform(self=Get HeadMesh(), InSocketName=FACIAL_R_Ear, TransformSpace=RTS_ParentBoneSpace))
   > Set FACIAL_R_Ear_Location(FACIAL_R_Ear_Location:Vector=@n25) -> (Output_Get:Vector)
   > Set FACIAL_R_Ear_Rotation(FACIAL_R_Ear_Rotation:Rotator=@n25.Rotation) -> (Output_Get:Rotator)
   n26 = BreakTransform(InTransform=GetSocketTransform(self=Get HeadMesh(), InSocketName=FACIAL_L_EyeCornerOuter, TransformSpace=RTS_ParentBoneSpace))
   > Set FACIAL_L_EyeCornerOuter_Location(FACIAL_L_EyeCornerOuter_Location:Vector=@n26) -> (Output_Get:Vector)
   > Set FACIAL_L_EyeCornerOuter_Rotation(FACIAL_L_EyeCornerOuter_Rotation:Rotator=@n26.Rotation) -> (Output_Get:Rotator)
   n27 = BreakTransform(InTransform=GetSocketTransform(self=Get HeadMesh(), InSocketName=FACIAL_R_EyeCornerOuter, TransformSpace=RTS_ParentBoneSpace))
   > Set FACIAL_R_EyeCornerOuter_Location(FACIAL_R_EyeCornerOuter_Location:Vector=@n27) -> (Output_Get:Vector)
   > Set FACIAL_R_EyeCornerOuter_Rotation(FACIAL_R_EyeCornerOuter_Rotation:Rotator=@n27.Rotation) -> (Output_Get:Rotator)
   n28 = BreakTransform(InTransform=GetSocketTransform(self=Get HeadMesh(), InSocketName=FACIAL_L_EyesackLower, TransformSpace=RTS_ParentBoneSpace))
   > Set FACIAL_L_EyesackLower_Location(FACIAL_L_EyesackLower_Location:Vector=@n28) -> (Output_Get:Vector)
   > Set FACIAL_L_EyesackLower_Rotation(FACIAL_L_EyesackLower_Rotation:Rotator=@n28.Rotation) -> (Output_Get:Rotator)
   n29 = BreakTransform(InTransform=GetSocketTransform(self=Get HeadMesh(), InSocketName=FACIAL_R_EyesackLower, TransformSpace=RTS_ParentBoneSpace))
   > Set FACIAL_R_EyesackLower_Location(FACIAL_R_EyesackLower_Location:Vector=@n29) -> (Output_Get:Vector)
   > Set FACIAL_R_EyesackLower_Rotation(FACIAL_R_EyesackLower_Rotation:Rotator=@n29.Rotation) -> (Output_Get:Rotator)
   n30 = BreakTransform(InTransform=GetSocketTransform(self=Get HeadMesh(), InSocketName=FACIAL_L_Masseter, TransformSpace=RTS_ParentBoneSpace))
   > Set FACIAL_L_Masseter_Location(FACIAL_L_Masseter_Location:Vector=@n30) -> (Output_Get:Vector)
   > Set FACIAL_L_Masseter_Rotation(FACIAL_L_Masseter_Rotation:Rotator=@n30.Rotation) -> (Output_Get:Rotator)
   n31 = BreakTransform(InTransform=GetSocketTransform(self=Get HeadMesh(), InSocketName=FACIAL_R_Masseter, TransformSpace=RTS_ParentBoneSpace))
   > Set FACIAL_R_Masseter_Location(FACIAL_R_Masseter_Location:Vector=@n31) -> (Output_Get:Vector)
   > Set FACIAL_R_Masseter_Rotation(FACIAL_R_Masseter_Rotation:Rotator=@n31.Rotation) -> (Output_Get:Rotator)
   n32 = BreakTransform(InTransform=GetSocketTransform(self=Get HeadMesh(), InSocketName=FACIAL_L_NasolabialBulge, TransformSpace=RTS_ParentBoneSpace))
   > Set FACIAL_L_NasolabialBulge_Location(FACIAL_L_NasolabialBulge_Location:Vector=@n32) -> (Output_Get:Vector)
   > Set FACIAL_L_NasolabialBulge_Rotation(FACIAL_L_NasolabialBulge_Rotation:Rotator=@n32.Rotation) -> (Output_Get:Rotator)
   n33 = BreakTransform(InTransform=GetSocketTransform(self=Get HeadMesh(), InSocketName=FACIAL_R_NasolabialBulge, TransformSpace=RTS_ParentBoneSpace))
   > Set FACIAL_R_NasolabialBulge_Location(FACIAL_R_NasolabialBulge_Location:Vector=@n33) -> (Output_Get:Vector)
   > Set FACIAL_R_NasolabialBulge_Rotation(FACIAL_R_NasolabialBulge_Rotation:Rotator=@n33.Rotation) -> (Output_Get:Rotator)
   n34 = BreakTransform(InTransform=GetSocketTransform(self=Get HeadMesh(), InSocketName=FACIAL_L_NasolabialFurrow, TransformSpace=RTS_ParentBoneSpace))
   > Set FACIAL_L_NasolabialFurrow_Location(FACIAL_L_NasolabialFurrow_Location:Vector=@n34) -> (Output_Get:Vector)
   > Set FACIAL_L_NasolabialFurrow_Rotation(FACIAL_L_NasolabialFurrow_Rotation:Rotator=@n34.Rotation) -> (Output_Get:Rotator)
   n35 = BreakTransform(InTransform=GetSocketTransform(self=Get HeadMesh(), InSocketName=FACIAL_R_NasolabialFurrow, TransformSpace=RTS_ParentBoneSpace))
   > Set FACIAL_R_NasolabialFurrow_Location(FACIAL_R_NasolabialFurrow_Location:Vector=@n35) -> (Output_Get:Vector)
   > Set FACIAL_R_NasolabialFurrow_Rotation(FACIAL_R_NasolabialFurrow_Rotation:Rotator=@n35.Rotation) -> (Output_Get:Rotator)
   Outputs()
  ]
 }
]
variables [
 bHasFaceBones {type=boolean}
 BodyMesh {type=SkeletalMeshComponentBudgeted}
 deform_upperarm_up_base_l_Location {type=Vector}
 deform_upperarm_up_base_l_Rotation {type=Rotator}
 deform_upperarm_up_base_r_Location {type=Vector}
 deform_upperarm_up_base_r_Rotation {type=Rotator}
 deform_upperarm_up_l_Location {type=Vector}
 deform_upperarm_up_l_Rotation {type=Rotator}
 deform_upperarm_up_r_Location {type=Vector}
 deform_upperarm_up_r_Rotation {type=Rotator}
 FACIAL_C_Chin_Location {type=Vector}
 FACIAL_C_Chin_Rotation {type=Rotator}
 FACIAL_C_Jaw_Location {type=Vector}
 FACIAL_C_Jaw_Rotation {type=Rotator}
 FACIAL_C_Jawline_Location {type=Vector}
 FACIAL_C_Jawline_Rotation {type=Rotator}
 FACIAL_C_LowerLipRotation_Location {type=Vector}
 FACIAL_C_LowerLipRotation_Rotation {type=Rotator}
 FACIAL_C_MouthLower_Location {type=Vector}
 FACIAL_C_MouthLower_Rotation {type=Rotator}
 FACIAL_C_MouthUpper_Location {type=Vector}
 FACIAL_C_MouthUpper_Rotation {type=Rotator}
 FACIAL_C_UnderChin_Lccation {type=Vector}
 FACIAL_C_UnderChin_Rotation {type=Rotator}
 FACIAL_L_CheekInner_Location {type=Vector}
 FACIAL_L_CheekInner_Rotation {type=Rotator}
 FACIAL_L_CheekLower_Location {type=Vector}
 FACIAL_L_CheekLower_Rotation {type=Rotator}
 FACIAL_L_CheekOuter_Location {type=Vector}
 FACIAL_L_CheekOuter_Rotation {type=Rotator}
 FACIAL_L_ChinSide_Location {type=Vector}
 FACIAL_L_ChinSide_Rotation {type=Rotator}
 FACIAL_L_Ear_Location {type=Vector}
 FACIAL_L_Ear_Rotation {type=Rotator}
 FACIAL_L_EyeCornerOuter_Location {type=Vector}
 FACIAL_L_EyeCornerOuter_Rotation {type=Rotator}
 FACIAL_L_EyesackLower_Location {type=Vector}
 FACIAL_L_EyesackLower_Rotation {type=Rotator}
 FACIAL_L_Jawline_Location {type=Vector}
 FACIAL_L_Jawline_Rotation {type=Rotator}
 FACIAL_L_LipCorner_Location {type=Vector}
 FACIAL_L_LipCorner_Rotation {type=Rotator}
 FACIAL_L_LipLowerOuter_Location {type=Vector}
 FACIAL_L_LipLowerOuter_Rotation {type=Rotator}
 FACIAL_L_Masseter_Location {type=Vector}
 FACIAL_L_Masseter_Rotation {type=Rotator}
 FACIAL_L_NasolabialBulge_Location {type=Vector}
 FACIAL_L_NasolabialBulge_Rotation {type=Rotator}
 FACIAL_L_NasolabialFurrow_Location {type=Vector}
 FACIAL_L_NasolabialFurrow_Rotation {type=Rotator}
 FACIAL_L_UnderChin_Location {type=Vector}
 FACIAL_L_UnderChin_Rotation {type=Rotator}
 FACIAL_R_CheekInner_Location {type=Vector}
 FACIAL_R_CheekInner_Rotation {type=Rotator}
 FACIAL_R_CheekLower_Location {type=Vector}
 FACIAL_R_CheekLower_Rotation {type=Rotator}
 FACIAL_R_CheekOuter_Location {type=Vector}
 FACIAL_R_CheekOuter_Rotation {type=Rotator}
 FACIAL_R_ChinSide_Location {type=Vector}
 FACIAL_R_ChinSide_Rotation {type=Rotator}
 FACIAL_R_Ear_Location {type=Vector}
 FACIAL_R_Ear_Rotation {type=Rotator}
 FACIAL_R_EyeCornerOuter_Location {type=Vector}
 FACIAL_R_EyeCornerOuter_Rotation {type=Rotator}
 FACIAL_R_EyesackLower_Location {type=Vector}
 FACIAL_R_EyesackLower_Rotation {type=Rotator}
 FACIAL_R_Jawline_Location {type=Vector}
 FACIAL_R_Jawline_Rotation {type=Rotator}
 FACIAL_R_LipCorner_Location {type=Vector}
 FACIAL_R_LipCorner_Rotation {type=Rotator}
 FACIAL_R_LipLowerOuter_Location {type=Vector}
 FACIAL_R_LipLowerOuter_Rotation {type=Rotator}
 FACIAL_R_Masseter_Location {type=Vector}
 FACIAL_R_Masseter_Rotation {type=Rotator}
 FACIAL_R_NasolabialBulge_Location {type=Vector}
 FACIAL_R_NasolabialBulge_Rotation {type=Rotator}
 FACIAL_R_NasolabialFurrow_Location {type=Vector}
 FACIAL_R_NasolabialFurrow_Rotation {type=Rotator}
 FACIAL_R_UnderChin_Location {type=Vector}
 FACIAL_R_UnderChin_Rotation {type=Rotator}
 Gravity Override {type=Vector}
 Gravity Override Params {type=GravityOverrideParamsStruct}
 HeadMesh {type=SkeletalMeshComponent}
 RBAN Params {
  type=RBANParamsStruct
  default {
   dBNO {simSpaceSettings {
     maxAngularAcceleration=1000
     maxAngularVelocity=3
     maxLinearAcceleration=300
     maxLinearVelocity=700
     velocityScaleZ=0
     externalLinearDragV=<0 0.5 0>
    }}
   dBNO_Carried {simSpaceSettings {
     maxAngularAcceleration=1000
     maxAngularVelocity=3
     maxLinearAcceleration=300
     maxLinearVelocity=700
     velocityScaleZ=0
     externalLinearDragV=<0 0.5 0>
    }}
   emoteOrMelee {simSpaceSettings {
     maxAngularAcceleration=1000
     maxAngularVelocity=3
     maxLinearAcceleration=300
     maxLinearVelocity=700
     velocityScaleZ=0
     externalLinearDragV=<0 0.5 0>
    }}
   front_End {simSpaceSettings {
     maxAngularAcceleration=1000
     maxAngularVelocity=3
     maxLinearAcceleration=300
     maxLinearVelocity=700
     velocityScaleZ=0
     externalLinearDragV=<0 0.5 0>
    }}
   onGround_Crouch_Moving {simSpaceSettings {
     maxAngularAcceleration=1000
     maxAngularVelocity=3
     maxLinearAcceleration=300
     maxLinearVelocity=700
     velocityScaleZ=0
     externalLinearDragV=<0 0.5 0>
    }}
   onGround_Falling {simSpaceSettings {
     maxAngularAcceleration=1000
     maxAngularVelocity=3
     maxLinearAcceleration=300
     maxLinearVelocity=700
     velocityScaleZ=0
     externalLinearDragV=<0 0.5 0>
    }}
   onGround_Moving {simSpaceSettings {
     maxAngularAcceleration=1000
     maxAngularVelocity=3
     maxLinearAcceleration=300
     maxLinearVelocity=700
     velocityScaleZ=0
     externalLinearDragV=<0 0.5 0>
    }}
   onGround_Standing {simSpaceSettings {
     maxAngularAcceleration=1000
     maxAngularVelocity=3
     maxLinearAcceleration=300
     maxLinearVelocity=700
     velocityScaleZ=0
     externalLinearDragV=<0 0.5 0>
    }}
   rail_Grinding {simSpaceSettings {
     maxAngularAcceleration=1000
     maxAngularVelocity=3
     maxLinearAcceleration=300
     maxLinearVelocity=700
     velocityScaleZ=0
     worldAlpha=0.1
     externalLinearDragV=<0 0.5 0>
    }}
   riding_Moving {simSpaceSettings {
     maxAngularAcceleration=1000
     maxAngularVelocity=3
     maxLinearAcceleration=300
     maxLinearVelocity=700
     velocityScaleZ=0
     externalLinearDragV=<0 0.5 0>
    }}
   riding_Moving_Targeting {simSpaceSettings {
     maxAngularAcceleration=1000
     maxAngularVelocity=3
     maxLinearAcceleration=300
     maxLinearVelocity=700
     velocityScaleZ=0
     externalLinearDragV=<0 0.5 0>
    }}
   riding_Standing {simSpaceSettings {
     maxAngularAcceleration=1000
     maxAngularVelocity=3
     maxLinearAcceleration=300
     maxLinearVelocity=700
     velocityScaleZ=0
     externalLinearDragV=<0 0.5 0>
    }}
   skydive_DivingDown {simSpaceSettings {
     maxAngularAcceleration=1000
     maxAngularVelocity=3
     maxLinearAcceleration=300
     maxLinearVelocity=700
     velocityScaleZ=0
     externalLinearDragV=<0 0.5 0>
    }}
   skydive_FloatingDown {
    componentLinearAccScale=<0 0.5 0>
    componentLinearVelScale=<0 0.5 0>
    simSpaceSettings {
     maxAngularAcceleration=1000
     maxAngularVelocity=3
     maxLinearAcceleration=300
     maxLinearVelocity=700
     velocityScaleZ=0
     externalLinearDragV=<0 0.5 0>
    }
   }
   skydive_GoingUp {simSpaceSettings {
     maxAngularAcceleration=1000
     maxAngularVelocity=3
     maxLinearAcceleration=300
     maxLinearVelocity=700
     velocityScaleZ=0
     externalLinearDragV=<0 0.5 0>
    }}
   skydive_Parachuting {
    componentLinearAccScale=<0 0.1 0.3>
    componentLinearVelScale=<0 0.1 0.3>
    simSpaceSettings {
     maxAngularAcceleration=1000
     maxAngularVelocity=3
     maxLinearAcceleration=300
     maxLinearVelocity=700
     velocityScaleZ=0
     externalLinearDragV=<0 0.5 0>
    }
   }
   skydive_Surfing {simSpaceSettings {
     maxAngularAcceleration=1000
     maxAngularVelocity=3
     maxLinearAcceleration=300
     maxLinearVelocity=700
     velocityScaleZ=0
     externalLinearDragV=<0 0.5 0>
    }}
   swimming {simSpaceSettings {
     maxAngularAcceleration=1000
     maxAngularVelocity=3
     maxLinearAcceleration=300
     maxLinearVelocity=700
     velocityScaleZ=0
     externalLinearDragV=<0 0.5 0>
    }}
   tactical_Sprint {simSpaceSettings {
     maxAngularAcceleration=1000
     maxAngularVelocity=3
     maxLinearAcceleration=300
     maxLinearVelocity=700
     velocityScaleZ=0
     externalLinearDragV=<0 0.5 0>
    }}
   ziplining {simSpaceSettings {
     maxAngularAcceleration=1000
     maxAngularVelocity=3
     maxLinearAcceleration=300
     maxLinearVelocity=700
     velocityScaleZ=0
     externalLinearDragV=<0 0.5 0>
    }}
  }
 }
 upperarm_l_Location {type=Vector}
 upperarm_l_Rotation {type=Rotator}
 upperarm_r_Location {type=Vector}
 upperarm_r_Rotation {type=Rotator}
 upperarm_twist_01_l_Location {type=Vector}
 upperarm_twist_01_l_Rotation {type=Rotator}
 upperarm_twist_01_r_Location {type=Vector}
 upperarm_twist_01_r_Rotation {type=Rotator}
]
